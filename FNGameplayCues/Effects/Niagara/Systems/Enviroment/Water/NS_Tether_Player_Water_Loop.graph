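# Niagara water effect loop for a tethered player moving on water, producing splashes, droplets, foam and ski-style jets.
# Splash and droplet particles spawn on the skeletal mesh, while two-group foam-jet meshes track its bones at locked height.
# Foam meshes snap to the water surface and stay constrained to its plane.
# User Speed scales spawning, lifetime or size per emitter, TurnSpeed pushes particles sideways, and the ContactSplashes emitter is disabled.
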NiagaraSystem /FNGameplayCues/Effects/Niagara/Systems/Enviroment/Water/NS_Tether_Player_Water_Loop
dependencies {
 usedDataInterfaces [
  &NiagaraDataInterfaceCurve
  &NiagaraDataInterfaceVector2DCurve
  /Script/Water.NiagaraDataInterfaceWater
  &NiagaraDataInterfaceVectorCurve
 ]
 usedDynamicInputs []
 usedModules [&SystemState]
}
emitters {
 ContactSplashes {
  bEnabled=false
  emitterUpdateScript [
   &EmitterState(Life Cycle Mode=::NewEnumerator1, Loop Behavior=::NewEnumerator1, Loop Duration Mode=::NewEnumerator1, UseLoopDelay=false)
   &SpawnRate(
    SpawnRate=&MakeCustomFloatFromBool(
     Bool=&SetBoolByFloatComparison(A=$User.Speed, B=0)
     True Float=&Lerp_Float(A=25, B=60, Alpha=$User.Speed)
     False Float=0
    )
    Use Spawn Probability=false
   )
  ]
  particleSpawnScript [
   /Game/Effects/Niagara/Modules/Groups/FN_SetSpawnGroup_v1/FN_SetSpawnGroup(Emitter Bone Count=2)
   /Niagara/Modules/Spawn/Initialization/InitializeParticle(
    Color=<rgba 1 1 1 2>
    Lifetime=&Lerp_Float(
     A=&UniformRangedFloat(Minimum=0.35, Maximum=0.45)
     B=&UniformRangedFloat(Minimum=0.2, Maximum=0.27)
     Alpha=$User.Speed
    )
    Mass=0.25
    Module.Write Color=true
    Module.Write Lifetime=true
    Module.Write Mass=true
    Module.Write Material Random=true
    Module.Write Position=false
    Module.Write Scale=false
    Module.Write SpriteRotation=false
    Module.Write SpriteSize=true
    Sprite Size=&Lerp_Vector2(
     A=&UniformRangedVector2D(Minimum=<18 25>, Maximum=<25 55>)
     B=&UniformRangedVector2D(Minimum=<20 100>, Maximum=<30 120>)
     Alpha=$User.Speed
    )
   )
   &SkeletalMeshLocation(
    Bone / Socket Index=$Particles.SpawnGroup
    Bone Sampling Mode=::NewEnumerator4
    Inherit Velocity=false
    MaskSpawnGroup=false
    Sampled Position Offset=&Lerp_Vector(
     A=&UniformRangedVector(Minimum=<2 45 0>, Maximum=<-2 0 0>)
     B=&UniformRangedVector(Minimum=<2 45 0>, Maximum=<-2 0 1>)
     Alpha=&MakeFloatFromInt(Int=$Particles.SpawnGroup)
    )
    Skeletal Mesh=$User.SkeletalMesh
   )
   &AddVelocity(Velocity=&MakeVector(
     X=&FloatFromCurve(
      FloatCurve=DataInterface() {propertyValues {Curve {keys [
          <at -1 -100 RCIM_Cubic>
          {arriveTangent=232.8515, interpMode=RCIM_Cubic, leaveTangent=232.7755, tangentMode=RCTM_User, time=0, value=0}
          <at 1 100>
         ]}}}
      CurveIndex=$User.TurnSpeed
      Scale Curve=&UniformRangedFloat(Minimum=1, Maximum=2)
     )
     Y=0
     Z=0
    ))
   &InheritVelocity(Module.Limit Speed=true, Velocity Limit=2000, Velocity Scale=<-0.1 -0.1 0>)
   &PointForce(
    Force Origin=&Add_Vector(A=/Niagara/DynamicInputs/Helpers/SimulationPosition(), B=<0 50 -25>)
    Force Strength=50
    Module.Use Falloff=false
    Module.UseLinearForce=false
   )
   &ApplyInitialForces()
  ]
  particleUpdateScript [
   &ParticleState()
   &Drag(Drag=2, Use Linear Drag=true, Use Rotational Drag=false)
   &GravityForce(Gravity=<0 0 -500>)
   &DynamicMaterialParameters(
    Index 0 Param 1=&FloatFromCurve(FloatCurve=DataInterface() {propertyValues {Curve {keys [<at -0.0046 -1.9941>,<at 1.0011 -1.3905>]}}})
    Module.Param0WriteEnabled=true
    Module.Param1WriteEnabled=true
    Module.Param2WriteEnabled=true
    Module.Param3WriteEnabled=true
   )
   &ScaleColor(
    Scale Alpha=&FloatFromCurve(FloatCurve=DataInterface() {propertyValues {Curve {keys [<at 0 1 RCIM_Cubic>, {arriveTangent=-2.1892, interpMode=RCIM_Cubic, leaveTangent=-2.1892, tangentMode=RCTM_User, time=1, value=0}]}}})
    ScaleA=true
    ScaleRGB=true
   )
   &ScaleSpriteSize(Scale Factor=&Vector2DFromCurve(
     Vector2Curve=DataInterface() {propertyValues {
       XCurve {keys [<at 0 0.5>,<at 1 1>]}
       YCurve {keys [
         {arriveTangent=1.8787, interpMode=RCIM_Cubic, leaveTangent=1.8787, tangentMode=RCTM_User, time=0, value=0.25}
         {arriveTangent=0.137, interpMode=RCIM_Cubic, leaveTangent=0.137, tangentMode=RCTM_User, time=1, value=1}
        ]}
      }}
     ScaleCurve=<1 1>
    ))
   &SolveForcesAndVelocity(Clamp Velocity=false, Limit Acceleration=false)
  ]
  rendererClasses [&NiagaraSpriteRendererProperties]
  renderers [{rendererClass=&NiagaraSpriteRendererProperties}]
 }
 FN_WaterSpash_SubUV {
  emitterUpdateScript [
   &EmitterScalability(
    Module.bApplySpawnCountScale=false
    Module.bApplySpawnCountScaleByDistanceFraction=false
    Module.bUseMaxDistance=false
    Module.bUseMinDistance=false
   ) {enabled=false}
   &EmitterLifeCycle() {enabled=false}
   &EmitterState()
   &SpawnRate(
    SpawnRate=&MakeCustomFloatFromBool(Bool=&SetBoolByFloatComparison(A=$User.Speed, B=0), True Float=25, False Float=0)
    Use Spawn Probability=false
   )
   &SpawnPerUnit(Module.Use Max Movement Threshold=true, Module.Use Movement Tolerance=true, Spawn Per Unit=75, Use Spawn Probability=false)
  ]
  particleSpawnScript [
   &InitializeParticle(
    Color=<rgba 0.5312 1.134 1.5 1.25>
    Lifetime=&Lerp_Float(
     A=&UniformRangedFloat(Minimum=0.5, Maximum=0.75)
     B=&UniformRangedFloat(Minimum=0.2, Maximum=0.4)
     Alpha=$User.Speed
    )
    Mass=&UniformRangedFloat(Minimum=1, Maximum=2)
    Module.Write Color=true
    Module.Write Lifetime=true
    Module.Write Mass=true
    Module.Write Material Random=true
    Module.Write Position=false
    Module.Write Scale=false
    Module.Write SpriteRotation=true
    Module.Write SpriteSize=true
    Position=/Niagara/Functions/Localspace/SimulationPosition()
    Sprite Rotation=&UniformRangedFloat(Minimum=0, Maximum=360)
    Sprite Size=&Vector2DFromFloat(Value=&UniformRangedFloat(Minimum=25, Maximum=70))
   )
   &AddVelocityInCone(
    Cone Angle=35
    Cone Axis=<0 0 1>
    Cone Axis Coordinate Space=::World
    Module.Use Velocity Falloff On Cone Axis=false
    Velocity Strength=&UniformRangedFloat(
     Minimum=50
     Maximum=&FloatFromCurve(FloatCurve=DataInterface() {propertyValues {Curve {keys [<at 0 1>,<at 0.1 2.3>]}}}, CurveIndex=$Emitter.NormalizedLoopAge, Scale Curve=700)
    )
   ) {enabled=false}
   &SkeletalMeshLocation(
    Bone Sampling Mode=::NewEnumerator0
    Inherit Velocity=false
    MaskSpawnGroup=false
    Module.Override Randomness=false
    Sampled Position Offset=&TransformVector(Vector=<0 35 15>, Destination Space=::Simulation)
    Skeletal Mesh=$User.SkeletalMesh
   )
   &AddVelocity(Velocity=&Add_Vector(
     A=&UniformRangedVector(Minimum=<-170 -170 120>, Maximum=<170 170 300>)
     B=&MakeVector(
      X=&FloatFromCurve(
       FloatCurve=DataInterface() {propertyValues {Curve {keys [
           <at -1 -100 RCIM_Cubic>
           {arriveTangent=125.2543, interpMode=RCIM_Cubic, leaveTangent=125.1909, tangentMode=RCTM_User, time=0, value=0}
           <at 1 100>
          ]}}}
       CurveIndex=$User.TurnSpeed
       Scale Curve=&UniformRangedFloat(Minimum=14, Maximum=20)
      )
      Y=0
      Z=0
     )
    ))
   /SetVariables(Particles.RandomGravity:NiagaraFloat=&UniformRangedFloat(Minimum=-3200, Maximum=-2600)) {isSetParameters=true}
  ]
  particleUpdateScript [
   &ParticleState()
   &AccelerationForce(Acceleration=&MakeVector(X=0, Y=0, Z=-2200)) {enabled=false}
   &Drag(Drag=0.5, Use Linear Drag=true, Use Rotational Drag=false)
   &GravityForce(Gravity=<0 0 -2200>)
   &SolveForcesAndVelocity(Clamp Velocity=false, Limit Acceleration=false)
   &Color() {enabled=false}
   &DynamicMaterialParameters(
    Index 0 Param 1=&FloatFromCurve(FloatCurve=DataInterface() {propertyValues {Curve {keys [<at 0.15 1 RCIM_Cubic>,<at 1 0>]}}})
    Module.Param0WriteEnabled=true
    Module.Param1WriteEnabled=true
    Module.Param2WriteEnabled=true
    Module.Param3WriteEnabled=true
   )
   &ScaleColor(
    Scale Alpha=&FloatFromCurve(
     FloatCurve=DataInterface() {propertyValues {Curve {keys [
         <at -0.0046 1.0113 RCIM_Cubic>
         {arriveTangent=-1.0036, interpMode=RCIM_Cubic, leaveTangent=-1.0036, time=0.8071, value=0.7176}
         <at 1.0011 0.002 RCIM_Cubic>
        ]}}}
     Scale Curve=0.5
    )
    ScaleA=true
    ScaleRGB=false
   )
   &ScaleSpriteSize(Scale Factor=&Vector2DFromFloat(Value=&FloatFromCurve(FloatCurve=DataInterface() {propertyValues {Curve {keys [
          {arriveTangent=2.3518, interpMode=RCIM_Cubic, leaveTangent=2.3518, tangentMode=RCTM_User, time=0, value=0.25}
          {arriveTangent=0.5711, interpMode=RCIM_Cubic, leaveTangent=0.5711, tangentMode=RCTM_User, time=1, value=2}
         ]}}})))
  ]
  rendererClasses [&NiagaraSpriteRendererProperties]
  renderers [{rendererClass=&NiagaraSpriteRendererProperties}]
 }
 FN_Water_Drops_ {
  emitterUpdateScript [
   &EmitterState()
   &EmitterScalability(
    MaxDistance=800
    Module.bApplySpawnCountScale=true
    Module.bApplySpawnCountScaleByDistanceFraction=true
    Module.bUseMaxDistance=true
    Module.bUseMinDistance=false
    SpawnCountScaleByDistanceFraction=DataInterface() {propertyValues {Curve {keys [<at 0 1>,<at 1 0.3>]}}}
   ) {enabled=false}
   &SpawnBurst_Instantaneous(Use Spawn Probability=false) {enabled=false}
   &SpawnPerUnit(
    Module.Use Max Movement Threshold=true
    Module.Use Movement Tolerance=true
    Movement Tolerance=15
    Spawn Per Unit:NiagaraFloat=&UniformRangedFloat(Minimum=40, Maximum=70)
    Use Spawn Probability=false
   )
   &EmitterLifeCycle(NextLoopDuration=0.25) {enabled=false}
   &SpawnRate(SpawnRate=55, Use Spawn Probability=false)
  ]
  particleSpawnScript [
   /SetVariables(Particles.Random=&UniformRangedFloat(Minimum=0, Maximum=1)) {isSetParameters=true}
   &InitializeParticle(
    Color=<rgba 1 1 1 0.8>
    Lifetime=&UniformRangedFloat(Minimum=0.6, Maximum=1.25)
    Mass=&UniformRangedFloat(Minimum=0.6, Maximum=1.2)
    Mesh Scale=<100 100 100>
    Module.Write Color=true
    Module.Write Lifetime=true
    Module.Write Mass=true
    Module.Write Material Random=true
    Module.Write Position=false
    Module.Write Scale=false
    Module.Write SpriteRotation=true
    Module.Write SpriteSize=true
    Position=&SimulationPosition()
    Sprite Size=&Vector2DFromCurve(
     Vector2Curve=DataInterface() {propertyValues {
       XCurve {keys [<at 0 1.5 RCIM_Cubic>,<at 0.9989 2 RCIM_Cubic>]}
       YCurve {keys [{arriveTangent=7.1794, time=0, value=4}, <at 0.9989 15 RCIM_Cubic>]}
      }}
     CurveIndex=$User.Speed
     ScaleCurve=<1 1>
    )
   )
   /Niagara/Modules/Spawn/Orientation/V2/InitialMeshOrientation(
    Orientation Vector=<0 0 0>
    Rotation=&UniformRangedVector(Minimum=<0 0 0>, Maximum=<0 0 1>)
    Use Orientation Vector=true
    Use Rotation Vector=true
   ) {enabled=false}
   &DynamicMaterialParameters(
    Index 0 Param 1=&UniformRangedFloat(Minimum=4, Maximum=10)
    Module.Param0WriteEnabled=true
    Module.Param1WriteEnabled=false
    Module.Param2WriteEnabled=false
    Module.Param3WriteEnabled=false
   )
   &InheritVelocity(
    Module.Limit Speed=true
    Velocity Limit=1000
    Velocity Scale=<-0.5 -0.5 -0.25>
    Velocity Source=&TransformVector(Vector=$Engine.Owner.Velocity, Source Space=::World, Destination Space=::Simulation)
   )
   &AddVelocity(Velocity=&Add_Vector(
     A=&UniformRangedVector(Minimum=<-50 -50 200>, Maximum=<50 50 450>)
     B=&TransformVector(
      Vector=&MakeVector(
       X=&FloatFromCurve(
        FloatCurve=DataInterface() {propertyValues {Curve {keys [
            <at -1 -100 RCIM_Cubic>
            {arriveTangent=234.6899, interpMode=RCIM_Cubic, leaveTangent=234.5927, tangentMode=RCTM_User, time=0, value=0}
            <at 1 100>
           ]}}}
        CurveIndex=$User.TurnSpeed
        Scale Curve=&UniformRangedFloat(Minimum=2, Maximum=8)
       )
       Y=0
       Z=0
      )
      Destination Space=::Simulation
     )
    ))
   &CylinderLocation(
    Cylinder Height=1
    Cylinder Radius=20
    Lathe Profile=DataInterface() {propertyValues {Curve {keys []}}}
    MaskSpawnGroup=false
    Module.Enable Lathe Profile=false
    Module.Override Randomness=false
    Module.Surface Only=true
    Offset=&MakeVector(
     X=0
     Y=25
     Z=&MakeFloatFromVector(
      Vector=&TransformVector(Vector=$User.WaterPlaneLocation, Source Space=::World, Destination Space=::Local)
      Channel=::NewEnumerator2
     )
    )
    Surface Only Band Thickness=2
   ) {enabled=false}
   &SkeletalMeshLocation(
    Bone Sampling Mode=::NewEnumerator0
    Inherit Velocity=false
    MaskSpawnGroup=false
    Module.Override Randomness=false
    Sampled Position Offset=&TransformVector(Vector=<0 35 15>, Destination Space=::Simulation)
    Skeletal Mesh=$User.SkeletalMesh
   )
   &InitializeWaterSnapToOwner() {enabled=false}
  ]
  particleUpdateScript [
   &ParticleState()
   /Game/Effects/Niagara/Emitters/Water/FN_DistanceKillParticles(AgeReduiceAmount=0.01)
   &ScaleColor(
    Scale Alpha=&FloatFromCurve(FloatCurve=DataInterface() {propertyValues {Curve {keys [
         {arriveTangent=27.637, interpMode=RCIM_Cubic, leaveTangent=27.637, tangentMode=RCTM_User, time=0, value=0}
         {arriveTangent=0.002, interpMode=RCIM_Cubic, leaveTangent=0.002, time=0.1183, value=1.0113}
         {arriveTangent=-1.5307, interpMode=RCIM_Cubic, leaveTangent=-1.5307, tangentMode=RCTM_User, time=1.0011, value=0.002}
        ]}}})
    ScaleA=true
    ScaleRGB=true
   )
   &PointForce(Force Origin=&SimulationPosition(), Force Strength=800, Module.Use Falloff=false, Module.UseLinearForce=false)
   &Drag(Use Linear Drag=true, Use Rotational Drag=false)
   &GravityForce(Gravity=<0 0 -1600>)
   &SolveForcesAndVelocity(Clamp Velocity=false, Limit Acceleration=false)
   &ScaleMeshSize() {enabled=false, module=MeshSizeScale}
   &ConstrainPositionToPlane(Plane Normal:Vector3f=$Particles.WaterPlaneNormal, Plane Position:Vector3f=$Particles.WaterPlaneLocation) {enabled=false}
  ]
  rendererClasses [&NiagaraSpriteRendererProperties, &NiagaraMeshRendererProperties]
  renderers [
   {rendererClass=&NiagaraSpriteRendererProperties}
   {rendererClass=&NiagaraMeshRendererProperties, rendererIndex=1}
  ]
 }
 FN_Water_Foam_SnapToSurface_V_Mask {
  emitterUpdateScript [
   &EmitterState()
   &EmitterScalability(
    MaxDistance=4000
    Module.bApplySpawnCountScale=true
    Module.bApplySpawnCountScaleByDistanceFraction=true
    Module.bUseMaxDistance=true
    Module.bUseMinDistance=false
    SpawnCountScaleByDistanceFraction=DataInterface() {propertyValues {Curve {keys [<at 0 1>,<at 1 0.5>]}}}
   ) {enabled=false}
   &SpawnPerUnit(
    Module.Use Max Movement Threshold=true
    Module.Use Movement Tolerance=true
    Movement Tolerance=5
    Spawn Per Unit=75
    Use Spawn Probability=false
   )
   &EmitterLifeCycle() {enabled=false}
   &SpawnRate(SpawnRate=3, Use Spawn Probability=false)
  ]
  particleSpawnScript [
   &FN_SetSpawnGroup(Emitter Bone Count=2)
   &InitializeParticle(
    Lifetime=&Lerp_Float(A=1.55, B=0.8, Alpha=$User.Speed)
    Mass=&UniformRangedFloat(Minimum=0.6, Maximum=1.2)
    Mesh Scale=&Lerp_Vector(
     A=&UniformRangedVector(Minimum=<100 80 100>, Maximum=<150 100 100>)
     B=&UniformRangedVector(Minimum=<350 120 100>, Maximum=<500 200 100>)
     Alpha=$User.Speed
    )
    Module.Write Color=true
    Module.Write Lifetime=true
    Module.Write Mass=true
    Module.Write Material Random=true
    Module.Write Position=true
    Module.Write Scale=true
    Module.Write SpriteRotation=false
    Module.Write SpriteSize=false
    Position=&Add_Vector(
     A=Linked() {linkedVariable {name=Engine.Owner.Position, type=NiagaraPosition}}
     B=&TransformVector(
      Vector=&MakeVector(
       X=&FloatFromCurve(
        FloatCurve=DataInterface() {propertyValues {Curve {keys [
            <at -0.5 -1 RCIM_Cubic>
            {arriveTangent=7.4715, interpMode=RCIM_Cubic, leaveTangent=7.4715, tangentMode=RCTM_User, time=0, value=0}
            <at 0.5 1>
           ]}}}
        CurveIndex=$User.TurnSpeed
        Scale Curve=60
       )
       Y=0
       Z=0
      )
      Destination Space=::Simulation
     )
    )
    Sprite Rotation=&NormalizedAngleToDegrees(NormalizedAngle=&UniformRangedFloat(Minimum=0, Maximum=1))
   ) {enabled=false}
   /Niagara/Modules/Spawn/Initialization/V2/InitializeParticle(
    Lifetime=&UniformRangedFloat(Minimum=2, Maximum=3)
    Mass=&UniformRangedFloat(Minimum=0.6, Maximum=1.2)
    Mass Mode=::NewEnumerator1
    Mesh Scale Mode=::NewEnumerator3
    Mesh Uniform Scale=100
    Sprite Rotation Angle=&UniformRangedFloat(Minimum=0, Maximum=1)
    Sprite Rotation Mode=::NewEnumerator1
    UsePositionOffset=false
    Write Material Random=true
   ) {module=InitializeParticle001}
   &InitialMeshOrientation(
    Orientation Vector=&Multiply_Vector(A=$Engine.Owner.Velocity, B=<1 1 0>)
    Rotation=&UniformRangedVector(Minimum=<0 0 0>, Maximum=<0 0 1>)
    Use Orientation Vector=true
    Use Rotation Vector=false
   ) {enabled=false}
   /Niagara/Modules/Spawn/Orientation/V3/InitialMeshOrientation(
    Mesh Orientation Mode=::NewEnumerator2
    Orientation Vector=<0 0 0>
    Rotation=&UniformRangedVector(Minimum=<0 0 0>, Maximum=<0 0 1>)
    Use Rotation Vector=true
   ) {module=InitialMeshOrientation001}
   &DynamicMaterialParameters(
    Index 1 Param 3=200
    Module.Param0WriteEnabled1=false
    Module.Param1WriteEnabled1=false
    Module.Param2WriteEnabled1=true
    Module.Param3WriteEnabled1=false
    Write Parameter Index 0=false
    Write Parameter Index 1=true
   )
   &AddVelocity(Velocity=&MakeVector(
     X=&FloatFromCurve(
      FloatCurve=DataInterface() {propertyValues {Curve {keys [
          <at -1 -100 RCIM_Cubic>
          {arriveTangent=234.6899, interpMode=RCIM_Cubic, leaveTangent=234.5927, tangentMode=RCTM_User, time=0, value=0}
          <at 1 100>
         ]}}}
      CurveIndex=$User.TurnSpeed
      Scale Curve=-5
     )
     Y=0
     Z=0
    ))
   &InitializeWaterSnapToOwner()
   &InheritVelocity(Module.Limit Speed=true, Velocity Limit=1000, Velocity Scale=<1 1 0>)
  ]
  particleUpdateScript [
   &ParticleState()
   &DynamicMaterialParameters(
    Index 0 Param 1=&FloatFromCurve(FloatCurve=DataInterface() {propertyValues {Curve {keys [<at -0.0046 1>,<at 1.0011 4>]}}})
    Index 0 Param 2=&FloatFromCurve(FloatCurve=DataInterface() {propertyValues {Curve {keys [<at -0.0046 0.5>,<at 1.0011 2>]}}})
    Index 0 Param 3=&FloatFromCurve(FloatCurve=DataInterface() {propertyValues {Curve {keys [<at -0.0046 1>,<at 1.0011 0.5>]}}})
    Index 2 Param 1=&FloatFromCurve(FloatCurve=DataInterface() {propertyValues {Curve {keys [<at 0 0.5>,<at 1.0011 1.5>]}}})
    Index 2 Param 2=&FloatFromCurve(FloatCurve=DataInterface() {propertyValues {Curve {keys [<at -0.0046 0.1>,<at 0.1441 2.105>,<at 1.0063 0.7129>]}}})
    Index 2 Param 3=1
    Index 2 Param 4=&FloatFromCurve(FloatCurve=DataInterface() {propertyValues {Curve {keys [
         {arriveTangent=-1.9991, interpMode=RCIM_Cubic, time=0, value=1}
         {arriveTangent=-3.1021, interpMode=RCIM_Cubic, leaveTangent=-3.1021, tangentMode=RCTM_User, time=0.2527, value=0.7055}
         <at 1 0 RCIM_Cubic>
        ]}}})
    Module.Param0WriteEnabled=true
    Module.Param0WriteEnabled2=true
    Module.Param1WriteEnabled=true
    Module.Param1WriteEnabled2=true
    Module.Param2WriteEnabled=true
    Module.Param2WriteEnabled2=true
    Module.Param3WriteEnabled=true
    Module.Param3WriteEnabled2=true
    Write Parameter Index 2=true
   ) {module=DynamicMaterialParameters001}
   &ScaleColor(
    Scale Alpha=&FloatFromCurve(FloatCurve=DataInterface() {propertyValues {Curve {keys [<at -0.0046 1.0113 RCIM_Cubic>,<at 1.0011 0.002 RCIM_Cubic>]}}})
    ScaleA=true
    ScaleRGB=true
   ) {enabled=false}
   &Drag(Drag=2.5, Use Linear Drag=true, Use Rotational Drag=false)
   &SolveForcesAndVelocity(Clamp Velocity=false, Limit Acceleration=false)
   &ScaleMeshSize(Scale Factor=&VectorFromFloat(Value=&FloatFromCurve(FloatCurve=DataInterface() {propertyValues {Curve {keys [<at 0 0.75>,<at 1 1.5>]}}}))) {module=MeshSizeScale}
   &ConstrainPositionToPlane(Plane Normal:Vector3f=$Particles.WaterPlaneNormal, Plane Position:Vector3f=$Particles.WaterPlaneLocation)
   &CameraOffset(Camera Offset Amount=25)
  ]
  rendererClasses [&NiagaraMeshRendererProperties]
  renderers [{rendererClass=&NiagaraMeshRendererProperties, rendererIndex=0}]
 }
 SkiFoamJet {
  emitterUpdateScript [
   &EmitterState(Life Cycle Mode=::NewEnumerator1, Loop Duration=0.25, UseLoopDelay=false)
   &EmitterScalability(
    MaxCulledState:ENiagaraExecutionState=::Active
    Module.bApplySpawnCountScale=false
    Module.bApplySpawnCountScaleByDistanceFraction=false
    Module.bUseMaxDistance=true
    Module.bUseMinDistance=false
   ) {enabled=false}
   &EmitterLifeCycle(NextLoopDuration=0.25) {enabled=false}
   &SpawnRate(SpawnRate=&Lerp_Float(A=17, B=30, Alpha=$User.Speed), Use Spawn Probability=false)
  ]
  particleSpawnScript [
   &FN_SetSpawnGroup(Emitter Bone Count=2)
   /Niagara/Modules/Spawn/Initialization/InitializeParticle(
    Lifetime=&Lerp_Float(A=0.9, B=0.4, Alpha=$User.Speed)
    Mesh Scale=&UniformRangedVector(Minimum=<0.75 0.4 0.5>, Maximum=<1.25 0.8 0.8>)
    Module.Write Color=true
    Module.Write Lifetime=true
    Module.Write Mass=true
    Module.Write Material Random=true
    Module.Write Position=false
    Module.Write Scale=true
    Module.Write SpriteRotation=false
    Module.Write SpriteSize=false
   )
   /SetVariables(
    Particles.MeshOrientation=<0 0 180 1>
    Particles.Position=&MakeVector(X=0, Y=0, Z=&MakeFloatFromVector(Vector=$User.WaterPlaneLocation, Channel=::NewEnumerator2))
    Particles.Scale=&Multiply_Vector(
     A=$Particles.Scale
     B=&Lerp_Vector(
      A=&UniformRangedVector(Minimum=<0.6 0.5 0.5>, Maximum=<1 0.65 0.6>)
      B=&UniformRangedVector(Minimum=<1.5 0.7 1>, Maximum=<2 0.85 1.2>)
      Alpha=$User.Speed
     )
    )
   ) {isSetParameters=true}
   /Game/Effects/Niagara/Emitters/Water/FN_WaterDisplace(WorldPosition:NiagaraPosition=$Particles.Position) {enabled=false}
   &AddVelocity(Velocity=&Add_Vector(
     A=&Lerp_Vector(A=<-20 0 0>, B=<20 0 0>, Alpha=&MakeFloatFromInt(Int=$Particles.SpawnGroup))
     B=&MakeVector(
      X=&FloatFromCurve(
       FloatCurve=DataInterface() {propertyValues {Curve {keys [
           <at -1 -100 RCIM_Cubic>
           {arriveTangent=232.8515, interpMode=RCIM_Cubic, leaveTangent=232.7755, tangentMode=RCTM_User, time=0, value=0}
           <at 1 100>
          ]}}}
       CurveIndex=$User.TurnSpeed
       Scale Curve=&UniformRangedFloat(Minimum=1.25, Maximum=2.5)
      )
      Y=0
      Z=0
     )
    ))
   &InheritVelocity(Module.Limit Speed=true, Velocity Limit=2000, Velocity Scale=<-0.1 -0.1 0>)
  ]
  particleUpdateScript [
   &ParticleState()
   &Drag(Drag=2, Use Linear Drag=true, Use Rotational Drag=false)
   &SkeletalMeshLocation(
    Bone / Socket Index=$Particles.SpawnGroup
    Bone Sampling Mode=::NewEnumerator4
    Inherit Velocity=false
    MaskSpawnGroup=false
    Sampled Position Offset=&TransformVector(
     Vector=&Lerp_Vector(A=<-5 25 0>, B=<5 25 0>, Alpha=&MakeFloatFromInt(Int=$Particles.SpawnGroup))
     Destination Space=::World
    )
    Skeletal Mesh=$User.SkeletalMesh
   )
   &ScaleColor(
    Scale Alpha=&FloatFromCurve(FloatCurve=DataInterface() {propertyValues {Curve {keys [
         {arriveTangent=9.9926, interpMode=RCIM_Cubic, leaveTangent=9.9927, tangentMode=RCTM_User, time=0, value=0}
         {arriveTangent=-2.8339, interpMode=RCIM_Cubic, leaveTangent=-2.8339, tangentMode=RCTM_User, time=0.0216, value=1}
         <at 1.0011 0.002 RCIM_Cubic>
        ]}}})
    ScaleA=true
    ScaleRGB=true
   )
   &SolveForcesAndVelocity(Clamp Velocity=false, Limit Acceleration=false)
   /SetVariables(Particles.Position=&MakeVector(
     X=&MakeFloatFromVector(Vector=Linked() {linkedVariable {name=Particles.Position, type=NiagaraPosition}}, Channel=::NewEnumerator0)
     Y=&MakeFloatFromVector(Vector=Linked() {linkedVariable {name=Particles.Position, type=NiagaraPosition}}, Channel=::NewEnumerator1)
     Z=&MakeFloatFromVector(Vector=Linked() {linkedVariable {name=Particles.Initial.Position, type=NiagaraPosition}}, Channel=::NewEnumerator2)
    )) {isSetParameters=true}
   /SetVariables(Particles.Scale=&Multiply_Vector(
     A=$Particles.Initial.Scale
     B=&VectorFromCurve(VectorCurve=DataInterface() {propertyValues {
        XCurve {keys [
          {arriveTangent=7.2936, interpMode=RCIM_Cubic, leaveTangent=7.2936, tangentMode=RCTM_User, time=0, value=0.05}
          {arriveTangent=1.1278, interpMode=RCIM_Cubic, leaveTangent=1.1278, tangentMode=RCTM_User, time=0.1463, value=0.6048}
          <at 0.6604 0.9268>
          {arriveTangent=0.2199, interpMode=RCIM_Cubic, leaveTangent=0.2199, tangentMode=RCTM_User, time=1, value=1}
         ]}
        YCurve {keys [
          {arriveTangent=1.5007, interpMode=RCIM_Cubic, leaveTangent=1.5007, tangentMode=RCTM_User, time=0, value=0.05}
          {arriveTangent=1.112, interpMode=RCIM_Cubic, leaveTangent=1.112, tangentMode=RCTM_User, time=0.5387, value=0.7547}
          {arriveTangent=3.438, interpMode=RCIM_Cubic, leaveTangent=3.438, tangentMode=RCTM_User, time=1, value=2}
         ]}
        ZCurve {keys [
          {arriveTangent=7.5652, interpMode=RCIM_Cubic, leaveTangent=7.5652, tangentMode=RCTM_User, time=0, value=0}
          {arriveTangent=-2.3619, interpMode=RCIM_Cubic, leaveTangent=-2.3619, tangentMode=RCTM_User, time=0.2477, value=0.9154}
          <at 0.6604 0.1922>
          <at 0.9989 0>
         ]}
       }})
    )) {isSetParameters=true}
   &DynamicMaterialParameters(
    Index 0 Param 1=&FloatFromCurve(FloatCurve=DataInterface() {propertyValues {Curve {keys [
         {arriveTangent=4.8859, interpMode=RCIM_Cubic, time=0, value=-1.7984}
         {arriveTangent=3.2109, interpMode=RCIM_Cubic, leaveTangent=3.2109, tangentMode=RCTM_User, time=1, value=0}
        ]}}})
    Index 0 Param 2=&Add_Float(
     A=&FloatFromCurve(FloatCurve=DataInterface() {propertyValues {Curve {keys [
          {arriveTangent=-9.4845, interpMode=RCIM_Cubic, leaveTangent=-9.4845, tangentMode=RCTM_User, time=0, value=4}
          {arriveTangent=0.2383, interpMode=RCIM_Cubic, leaveTangent=0.2383, tangentMode=RCTM_User, time=1, value=0}
         ]}}})
     B=0
    )
    Index 1 Param 1=&FloatFromCurve(FloatCurve=DataInterface() {propertyValues {Curve {keys [
         {arriveTangent=0.5951, interpMode=RCIM_Cubic, time=0, value=0}
         {arriveTangent=2.0085, interpMode=RCIM_Cubic, leaveTangent=2.0085, tangentMode=RCTM_User, time=1, value=1}
        ]}}})
    Module.Param0WriteEnabled=true
    Module.Param0WriteEnabled1=true
    Module.Param0WriteEnabled3=true
    Module.Param1WriteEnabled=true
    Module.Param1WriteEnabled1=true
    Module.Param1WriteEnabled3=true
    Module.Param2WriteEnabled=true
    Module.Param2WriteEnabled1=true
    Module.Param2WriteEnabled3=true
    Module.Param3WriteEnabled=true
    Module.Param3WriteEnabled1=true
    Module.Param3WriteEnabled3=true
    Write Parameter Index 1=true
    Write Parameter Index 3=true
   )
   /Niagara/Modules/Update/Orientation/V2/UpdateMeshOrientation(Rotation Vector:Vector3f=&MakeVector(X=&Multiply_Float(A=$User.TurnSpeed, B=-1), Y=0, Z=0))
  ]
  rendererClasses [&NiagaraMeshRendererProperties]
  renderers [{rendererClass=&NiagaraMeshRendererProperties, rendererIndex=0}]
 }
}
summary {
 systemName=NS_Tether_Player_Water_Loop
 emitters [
  {
   emitterName=FN_Water_Drops_
   rendererClasses [&NiagaraSpriteRendererProperties, &NiagaraMeshRendererProperties]
  }
  {emitterName=FN_Water_Foam_SnapToSurface_V_Mask, rendererClasses [&NiagaraMeshRendererProperties]}
  {emitterName=SkiFoamJet, rendererClasses [&NiagaraMeshRendererProperties]}
  {emitterName=FN_WaterSpash_SubUV, rendererClasses [&NiagaraSpriteRendererProperties]}
  {bEnabled=false, emitterName=ContactSplashes, rendererClasses [&NiagaraSpriteRendererProperties]}
 ]
 userVariables [
  User.Boost {type=NiagaraFloat, defaultValue=0, description=""}
  User.ForwardSpeed {type=NiagaraFloat, defaultValue=0, description=""}
  User.SkeletalMesh {
   type=NiagaraDataInterfaceSkeletalMesh
   defaultValue {
    struct=NiagaraExt_VariableValue_DataInterface
    value {dataInterface=&self:NiagaraDataInterfaceSkeletalMesh_0}
   }
  }
  User.Speed {type=NiagaraFloat, defaultValue=0.5, description=""}
  User.TurnSpeed {type=NiagaraFloat, defaultValue=0, description=""}
  User.WaterBodyIndex {type=NiagaraFloat, defaultValue=0, description=""}
  User.WaterDepth {type=NiagaraFloat, defaultValue=500, description=""}
  User.WaterPlaneLocation {type=Vector3f, description="", defaultValue {struct=Vector3f, value=<0 0 352>}}
  User.WaterPlaneNormal {type=Vector3f, description="", defaultValue {struct=Vector3f, value=<0 0 1>}}
  User.WaveAttenuationFactor {type=NiagaraFloat, defaultValue=0, description=""}
  User.WaveTime {type=NiagaraFloat, defaultValue=0, description=""}
 ]
}
